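# Animation Blueprint for the FossilMeal_Bread character on the Fortnite player skeleton, supplying MovementMode layers that change its walk while patrolling.
# Only the SourcePose layer has logic: nested bool blends play a looping WalkHeels sequence when IsPatroling and bIsMoving2D are true, otherwise passing through the cached input pose.
# The other MovementMode layers, AnimGraph and EventGraph contain only unconnected nodes, and nothing here sets IsPatroling.
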AnimBlueprint /FossilMeal_Bread/Assets/Animations/FossilMeal_Bread_AnimBP
target_skeleton=&Fortnite_M_Avg_Player_Skeleton
graphs [
 AnimGraph {nodes [Output Pose()]}
 EventGraph {nodes [
   TryGetPawnOwner()
   Event BlueprintUpdateAnimation()
  ]}
 MovementMode_AimOffset {nodes [
   Input Pose() {name=InAimOffsetsPose}
   Input Pose() {name=InSourcePosePreAimOffsets}
   Output Pose() {layerGroup=MovementMode}
  ]}
 MovementMode_BaseLocomotion {nodes [
   Input Pose() {name=InPoseBaseLocomotion}
   Output Pose() {layerGroup=MovementMode}
  ]}
 MovementMode_FinalPoseOverride {nodes [
   Input Pose() {name=InFinalPose}
   Output Pose() {layerGroup=MovementMode}
  ]}
 MovementMode_FullBodyOverride {nodes [
   Input Pose() {name=InputPoseFullBody}
   Output Pose() {layerGroup=MovementMode}
  ]}
 MovementMode_InAirOverride {nodes [
   Input Pose() {name=InSource_Pose}
   Input Pose() {name=InAirPose}
   Output Pose() {layerGroup=MovementMode}
  ]}
 MovementMode_LocomotionAdditives {nodes [
   Input Pose() {name=InLocomotionAdditivesPose}
   Output Pose() {layerGroup=MovementMode}
  ]}
 MovementMode_LowerBodyOverride {nodes [
   Input Pose() {name=InLowerBodyPose}
   Input Pose() {name=InSourcePose}
   Output Pose() {layerGroup=MovementMode}
  ]}
 MovementMode_OffScreen {nodes [
   Input Pose() {name=InPoseOffscreen}
   Output Pose() {layerGroup=MovementMode}
  ]}
 MovementMode_SourcePose {nodes [
   n1 = Sequence Player 'FossilMeal_Bread_WalkHeels_Loop_N_CMF'() {sequence=&/FossilMeal_Bread_WalkHeels_Loop_N_CMF}
   n2 = Use cached pose 'MoveSourcePoseInput'()
   n3 = Get bIsMoving2D()
   Output Pose(Result=Blend Poses by bool(
     BlendPose_0=Blend Poses by bool(BlendPose_0=@n1, BlendPose_1=@n2, bActiveValue=@n3)
     BlendPose_1=Use cached pose 'MoveSourcePoseInput'()
     bActiveValue=Get IsPatroling() -> (IsPatroling:bool)
    )) {layerGroup=MovementMode}
   Save cached pose 'MoveSourcePoseInput'(Pose=Input Pose() {name=InSourcePoseBase}) {cacheName=MoveSourcePoseInput}
  ]}
]
variables [IsPatroling {type=boolean}]
