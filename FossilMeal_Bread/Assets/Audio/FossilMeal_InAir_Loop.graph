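# FossilMeal_InAir_Loop is a looping sound cue for in-air falling wind, built from borrowed Athena skydive and character falling sounds.
# A mixer layers five looping waves, including an audio tell and a doubled enemy wind layer, into an enveloper that ramps volume up.
# An attenuation node adds HRTF spatialization, custom falloff and occlusion before the root looping node.
# The cue also overrides its own attenuation and limits concurrency, stopping the quietest instance.
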SoundCue /FossilMeal_Bread/Assets/Audio/FossilMeal_InAir_Loop
root=Looping_0
nodes [
 {
  id=Attenuation_1
  children [Enveloper_7]
  properties {
   bOverrideAttenuation=true
   attenuationOverrides {
    bEnableOcclusion=true
    bEnableReverbSend=false
    distanceAlgorithm=Custom
    falloffDistance=3250
    hPFFrequencyAtMax=300
    lPFFrequencyAtMax=800
    lPFRadiusMax=3000
    lPFRadiusMin=1000
    occlusionInterpolationTime=0.5
    occlusionLowPassFilterFrequency=6000
    occlusionTraceChannel=ECC_GameTraceChannel11
    occlusionVolumeAttenuation=3
    spatializationAlgorithm=SPATIALIZATION_HRTF
    stereoSpread=0
    attenuationShapeExtents {x=1000}
    customAttenuationCurve {editorCurveData {keys [
       <at 0 1 RCIM_Cubic>
       {arriveTangent=-1.7422, interpMode=RCIM_Cubic, leaveTangent=-1.7422, time=0.1205, value=0.8944}
       {arriveTangent=-1.017, interpMode=RCIM_Cubic, leaveTangent=-1.017, time=0.3914, value=0.3181}
       {arriveTangent=-0.5016, interpMode=RCIM_Cubic, leaveTangent=-0.5016, tangentMode=RCTM_User, time=1, value=0}
      ]}}
   }
  }
 }
 {
  id=Enveloper_7
  children [Mixer_4]
  properties {
   pitchCurve {editorCurveData {keys [{time=0}]}}
   volumeCurve {editorCurveData {keys [
      <at 0 1>
      {arriveTangent=6.1531, interpMode=RCIM_Cubic, leaveTangent=6.1531, tangentMode=RCTM_User, time=0, value=0.007}
      {arriveTangent=0.9931, interpMode=RCIM_Cubic, leaveTangent=0.9931, tangentMode=RCTM_User, time=0.1837, value=0.7651}
      <at 1 1>
     ]}}
  }
 }
 {id=Looping_0, children [Attenuation_1]}
 {id=Mixer_4, children [WavePlayer_4, WavePlayer_1, WavePlayer_13, WavePlayer_12, WavePlayer_0], properties {inputVolume [1,0.6,1,1,1]}}
 {id=WavePlayer_0, properties {bLooping=true, soundWaveAssetPtr=/Game/Athena/Sounds/Character/Falling/Character_InAir_Wind_Lp_Enemy_01}}
 {id=WavePlayer_1, properties {bLooping=true, soundWaveAssetPtr=/Game/Athena/Sounds/Wind_Falling/Character_InAir_Wind_Lp_02}}
 {id=WavePlayer_12, properties {bLooping=true, soundWaveAssetPtr=&Character_InAir_Wind_Lp_Enemy_01}}
 {id=WavePlayer_13, properties {bLooping=true, soundWaveAssetPtr=/Game/Athena/Sounds/Wind_Falling/Character_InAir_AudioTell_01}}
 {id=WavePlayer_4, properties {bLooping=true, soundWaveAssetPtr=/Game/Athena/Sounds/Wind_Falling/AMB_Wind_Skydive_Vertical_Loop_01}}
]
properties {
 bOverrideAttenuation=true
 bOverrideConcurrency=true
 maxDistance=7000
 volumeMultiplier=1.4
 assetUserData [&self:AudioAssetUserData_0]
 attenuationOverrides {
  bAttenuateWithLPF=true
  bEnableListenerFocus=true
  bEnableOcclusion=true
  falloffDistance=6000
  lPFFrequencyAtMax=5000
  occlusionLowPassFilterFrequency=2500
  occlusionTraceChannel=ECC_GameTraceChannel11
  attenuationShapeExtents {x=1000}
 }
 concurrencyOverrides {maxCount=3, resolutionRule=StopQuietest}
}
