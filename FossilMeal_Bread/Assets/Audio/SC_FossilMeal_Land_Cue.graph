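# Sound cue that plays the FossilBread landing sound for the FossilMeal Bread feature, routed to the boss fight SFX sound class.
# A single wave player feeds a modulator that randomizes volume on each play.
# The cue overrides attenuation with a custom falloff curve that drops steeply early, plus distance-based low-pass filtering, listener focus, and occlusion that muffles and lowers blocked sound.
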SoundCue /FossilMeal_Bread/Assets/Audio/SC_FossilMeal_Land_Cue
root=Modulator_0
nodes [
 {id=Modulator_0, children [WavePlayer_2], properties {volumeMax=1.1, volumeMin=0.8}}
 {id=WavePlayer_2, properties {soundWaveAssetPtr=/FossilMeal_Bread/Assets/Audio/waves/FossilBread_Land}}
]
properties {
 bOverrideAttenuation=true
 maxDistance=5000
 soundClassObject=&Fort_BossFight_SFX
 volumeMultiplier=1.2
 attenuationOverrides {
  bAttenuateWithLPF=true
  bEnableListenerFocus=true
  bEnableOcclusion=true
  distanceAlgorithm=Custom
  falloffDistance=4000
  lPFFrequencyAtMax=2000
  lPFRadiusMax=1000
  lPFRadiusMin=500
  occlusionLowPassFilterFrequency=1800
  occlusionTraceChannel=ECC_GameTraceChannel11
  occlusionVolumeAttenuation=0.7
  attenuationShapeExtents {x=1000}
  customAttenuationCurve {editorCurveData {keys [<at 0 1>,<at 0.0985 0.9361>,<at 0.2745 0.2765>,<at 0.8108 0.1122>,<at 1 0>]}}
 }
}
