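# Sound cue for the FossilMeal Bread pre-jump, routed to the boss fight SFX sound class.
# A Mixer layers a quieter PreJump wave, the full-level Jump wave and a random pick from three alert variants.
# The PreJump layer has its own attenuation node with a shorter custom falloff.
# Cue-level attenuation override adds a longer custom curve, distance low-pass, occlusion and listener focus.
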SoundCue /FossilMeal_Bread/Assets/Audio/SC_FossilMeal_PreJump_Cue
root=Mixer_0
nodes [
 {
  id=Attenuation_0
  children [WavePlayer_0]
  properties {
   bOverrideAttenuation=true
   attenuationOverrides {
    distanceAlgorithm=Custom
    falloffDistance=3000
    occlusionLowPassFilterFrequency=2500
    occlusionTraceChannel=ECC_GameTraceChannel11
    attenuationShapeExtents {x=50}
    customAttenuationCurve {editorCurveData {keys [
       <at -0.0028 0.7079>
       <at 0.0886 0.6701>
       <at 0.21 0.4233>
       {arriveTangent=-0.7422, leaveTangent=-0.7422, time=0.4987, value=0.2917}
       <at 0.9881 0.2343>
       <at 1.0024 0.0098>
      ]}}
   }
  }
 }
 {id=Mixer_0, children [Attenuation_0, WavePlayer_1, Random_0], properties {inputVolume [0.3,1,0.6]}}
 {id=Random_0, children [WavePlayer_3,WavePlayer_4,WavePlayer_5], properties {weights [1,1,1]}}
 {id=WavePlayer_0, properties {soundWaveAssetPtr=/FossilMeal_Bread/Assets/Audio/waves/FossilBread_PreJump}}
 {id=WavePlayer_1, properties {soundWaveAssetPtr=/FossilMeal_Bread/Assets/Audio/waves/FossilBread_Jump}}
 {id=WavePlayer_3, properties {soundWaveAssetPtr=/FossilMeal_Bread/Assets/Audio/waves/FossilBread_PreJump_Alert_01}}
 {id=WavePlayer_4, properties {soundWaveAssetPtr=/FossilMeal_Bread/Assets/Audio/waves/FossilBread_PreJump_Alert_02}}
 {id=WavePlayer_5, properties {soundWaveAssetPtr=/FossilMeal_Bread/Assets/Audio/waves/FossilBread_PreJump_Alert_03}}
]
properties {
 bOverrideAttenuation=true
 maxDistance=2097152
 soundClassObject=&Fort_BossFight_SFX
 volumeMultiplier=1.4
 attenuationOverrides {
  bAttenuateWithLPF=true
  bEnableListenerFocus=true
  bEnableOcclusion=true
  distanceAlgorithm=Custom
  falloffDistance=6500
  lPFFrequencyAtMax=5000
  lPFRadiusMax=3500
  lPFRadiusMin=500
  occlusionLowPassFilterFrequency=1800
  occlusionTraceChannel=ECC_GameTraceChannel11
  occlusionVolumeAttenuation=0.7
  attenuationShapeExtents {x=50}
  customAttenuationCurve {editorCurveData {keys [<at 0.0024 0.7218>,<at 0.0515 0.6647>,<at 0.1401 0.3821>,<at 0.5074 0.3255>,<at 0.9832 0.2752>,<at 1 0>]}}
 }
}
