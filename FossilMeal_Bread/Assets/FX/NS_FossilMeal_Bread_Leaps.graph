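# Niagara impact effect for FossilMeal Bread leaps: a short burst of dark streaks with a flare and smoke.
# DarkStreaks throws drag-slowed sprites sized by User.Scalar; DarkBits is disabled.
# FlareSmoke spawns at System.ImpactOffset, driving two sprite renderers and a mesh renderer through separate color, size and material-parameter attributes.
# System.IsWanderingBoss hue-shifts and desaturates the main color, and System.DistanceScalarNear adjusts the dark mesh's opacity.
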NiagaraSystem /FossilMeal_Bread/Assets/FX/NS_FossilMeal_Bread_Leaps
dependencies {
 usedDataInterfaces [
  &NiagaraDataInterfaceArrayColor
  &NiagaraDataInterfaceArrayFloat2
  &NiagaraDataInterfaceCurve
  &NiagaraDataInterfaceColorCurve
  &NiagaraDataInterfaceVector2DCurve
 ]
 usedDynamicInputs [&FN_LOD_Distance_Remap]
 usedModules [
  &SystemState
  &self:SystemSpawnScript.SystemScriptSource.SystemScriptGraph.NiagaraNodeAssignment_0.SetVariables_#1
 ]
}
emitters {
 DarkBits {
  bEnabled=false
  emitterUpdateScript [
   &EmitterState(Life Cycle Mode=::NewEnumerator1, Loop Behavior=::NewEnumerator1, Loop Duration=0.16, UseLoopDelay=true)
   &SpawnRate(SpawnRate=150, Use Spawn Probability=false)
  ]
  particleSpawnScript [
   &InitializeParticle(
    Lifetime Max=0.8
    Lifetime Min=0.3
    Lifetime Mode=::NewEnumerator1
    Mass Mode=::NewEnumerator3
    Sprite Rotation Mode=::NewEnumerator2
    Sprite Size Mode=::NewEnumerator1
    Uniform Sprite Size Max=20
    Uniform Sprite Size Min=6
    UsePositionOffset=false
    Write Material Random=true
   )
   &ShapeLocation(
    MaskSpawnGroup=false
    Non Uniform Scale=&VectorFromFloat(Value=$User.Scalar)
    Offset=<0 0 30>
    Offset Mode=::NewEnumerator0
    Override Seed=false
    Sphere Radius=60
    UseOwnerScale=true
   )
   &AddVelocity(
    Module.Use Falloff Distance=false
    Override Seed=false
    Velocity Mode=::NewEnumerator1
    Velocity Origin=$Particles.ShapeLocation.ShapePosition
    Velocity Speed=&Multiply_Float(A=$User.Scalar, B=&RandomRangeFloat(Minimum=250, Maximum=600))
   )
   &AddVelocity(Velocity=<0 0 150>, Velocity Speed Scale=&RandomRangeFloat(Minimum=0.5, Maximum=1)) {module=AddVelocity001}
  ]
  particleUpdateScript [
   &ParticleState()
   &GravityForce(Gravity=<0 0 -350>)
   &ScaleColor(
    Scale Alpha=&FloatFromCurve(FloatCurve=DataInterface() {propertyValues {Curve {keys [<at 0.5 1 RCIM_Cubic>,<at 1 0 RCIM_Cubic>]}}})
    Scale RGB=&MakeVectorFromLinearColorRGB(Color=&ColorFromCurve(ColorCurve=DataInterface() {propertyValues {
        AlphaCurve {keys [<at 0 1 RCIM_Cubic>]}
        BlueCurve {keys [<at 0.058 33.3333 RCIM_Cubic>,<at 0.4898 0.4456 RCIM_Cubic>]}
        GreenCurve {keys [<at 0.058 0 RCIM_Cubic>,<at 0.4898 0.2081 RCIM_Cubic>]}
        RedCurve {keys [<at 0.058 200 RCIM_Cubic>,<at 0.4898 1 RCIM_Cubic>]}
       }}))
    ScaleA=true
    ScaleRGB=true
   )
   &CurlNoiseForce(
    Bias Noise=false
    Noise Frequency=20
    Noise Strength=&Multiply_Float(
     A=700
     B=&FloatFromCurve(FloatCurve=DataInterface() {propertyValues {Curve {keys [<at 0.4 1>,<at 1 0>]}}})
    )
    Pan Noise=true
    Pan Noise Field=<0.1 1 0.1>
   )
   &Drag(Drag=&Multiply_Float(A=2, B=$User.Scalar), Use Linear Drag=true, Use Rotational Drag=false)
   &SolveForcesAndVelocity(Clamp Velocity=false, Limit Acceleration=false)
  ]
  rendererClasses [&NiagaraSpriteRendererProperties]
  renderers [{rendererClass=&NiagaraSpriteRendererProperties}]
 }
 DarkStreaks {
  emitterUpdateScript [
   &EmitterState(Life Cycle Mode=::NewEnumerator1, Loop Behavior=::NewEnumerator1, Loop Duration=1, UseLoopDelay=false)
   &SpawnBurst_Instantaneous(Spawn Count=15, Use Spawn Probability=false, UseLoopCountLimit=false)
  ]
  particleSpawnScript [
   &PartitionParticles()
   /SetVariables(Particles.RandomFloat=&UniformRangedFloat(Minimum=0, Maximum=3)) {isSetParameters=true}
   &InitializeParticle(
    Color=&SelectLinearColorFromArray(
     Array Sampling Mode=::NewEnumerator1
     Direct Array Index=$Particles.PartitionParticles.Partition
     Color Selection Array=DataInterface() {propertyValues {ColorData [<rgba 0.045 0.0458 0.05 0.5>, <rgba 0.27 0.2885 0.3 0.8>]}}
     Random Seed=?
    )
    Lifetime Max=0.5
    Lifetime Min=0.2
    Lifetime Mode=::NewEnumerator1
    Sprite Size=&Multiply_Vector2D_ByFloat(
     Float=&RandomRangeFloat(Minimum=1, Maximum=1.5)
     Vector2D=&SelectVector2DFromArray(
      Direct Array Index=$Particles.PartitionParticles.Partition
      Vector 2D Selection Array=DataInterface() {propertyValues {FloatData [<30 50>, <15 70>]}}
     )
    )
    Sprite Size Mode=::NewEnumerator4
    UsePositionOffset=false
    Write Material Random=true
   ) {module=InitializeParticle001}
   &ShapeLocation(
    Handle Radius=5
    Large Radius=5
    MaskSpawnGroup=false
    Non Uniform Scale=<1 1 0.1>
    Override Seed=false
    Shape Primitive=::NewEnumerator3
    UseOwnerScale=true
   ) {module=ShapeLocation001}
   &AddVelocity(
    Module.Use Falloff Distance=false
    Override Seed=false
    Velocity Mode=::NewEnumerator1
    Velocity Speed=&Lerp_Float(
     A=700
     B=1000
     Alpha=&FloatFromCurve(
      FloatCurve=DataInterface() {propertyValues {Curve {keys [{interpMode=RCIM_Cubic, tangentMode=RCTM_Break, time=0, value=0}, {interpMode=RCIM_Cubic, tangentMode=RCTM_Break, time=1, value=1}]}}}
      CurveIndex=&RandomRangeFloat(Minimum=0, Maximum=1)
      Scale Curve=0.5
     )
    )
   )
   &AddVelocity(Velocity=&RandomRangeVector(Minimum=<0 0 100>, Maximum=<0 0 500>)) {module=AddVelocity001}
   &SubUVAnimation(End Frame=3, SubUV Animation Mode:ENiagara_SubUVLookupMode=::NewEnumerator2)
  ]
  particleUpdateScript [
   &ParticleState() {module=UpdateAge}
   &ApplyOwnerScaleToAttributes()
   &Drag(Drag=6, Use Linear Drag=true, Use Rotational Drag=false)
   &SolveForcesAndVelocity(Clamp Velocity=false, Limit Acceleration=false)
   &ScaleSpriteSize(Scale Factor=&Vector2DFromFloat(Value=&FloatFromCurve(
      FloatCurve=DataInterface() {propertyValues {Curve {keys [
          {arriveTangent=1.6317, interpMode=RCIM_Cubic, leaveTangent=1.6317, tangentMode=RCTM_User, time=0, value=0.1}
          {arriveTangent=0.2754, interpMode=RCIM_Cubic, leaveTangent=0.2754, tangentMode=RCTM_User, time=1.0463, value=1}
         ]}}}
      Scale Curve=&Multiply_Float(A=2, B=$User.Scalar)
     )))
   &ScaleColor(
    Scale Alpha=&FloatFromCurve(
     FloatCurve=DataInterface() {propertyValues {Curve {keys [{interpMode=RCIM_Cubic, tangentMode=RCTM_Break, time=0, value=1}, {interpMode=RCIM_Cubic, tangentMode=RCTM_Break, time=1, value=0}]}}}
     Scale Curve=&RandomRangeFloat(Minimum=2, Maximum=2.5)
    )
    ScaleA=true
    ScaleRGB=false
   )
   &DynamicMaterialParameters(
    Index 0 Param 2=1
    Index 0 Param 3=1
    Index 0 Param 4=&FloatFromCurve(FloatCurve=DataInterface() {propertyValues {Curve {keys [<at 0.3 1>,<at 1 0.25>]}}})
    Module.Param0WriteEnabled=true
    Module.Param1WriteEnabled=true
    Module.Param2WriteEnabled=true
    Module.Param3WriteEnabled=true
   )
  ]
  rendererClasses [&NiagaraSpriteRendererProperties]
  renderers [{rendererClass=&NiagaraSpriteRendererProperties}]
 }
 FlareSmoke {
  emitterUpdateScript [
   &EmitterState(Life Cycle Mode=::NewEnumerator1, Loop Behavior=::NewEnumerator1, Loop Duration=2, UseLoopDelay=true)
   &SpawnBurst_Instantaneous(Use Spawn Probability=false, UseLoopCountLimit=false)
  ]
  particleSpawnScript [
   &InitializeParticle(
    Lifetime=0.7
    Mass Mode=::NewEnumerator3
    Mesh Scale=&Multiply_VectorByFloat(Vector=<1 1 1>, Float=0.25)
    Mesh Scale Mode=::NewEnumerator4
    Sprite Size=<900 500>
    Sprite Size Mode=::NewEnumerator4
    UsePositionOffset=false
    Write Material Random=true
   )
   &SystemLocation(
    MaskSpawnGroup=false
    Offset=&AddVectorToPosition(Position=$System.ImpactOffset, Vector=<0 0 -3>)
    Offset Coordinate Space=::World
   )
   &CameraOffset(Camera Offset Amount=60)
   &InitialMeshOrientation(Orientation Axis=<0 0 1>, Orientation Vector=<0 0 1>, Rotation=<0 0.25 0>, Use Orientation Vector=true, Use Rotation Vector=true)
   &InitialMeshOrientation(
    Orientation Axis=<0 0 0>
    Orientation Vector=<0 0 0>
    Rotation=&RandomRangeVector(Minimum=<0 0 0>, Maximum=<1 0 0>)
    Use Orientation Vector=false
    Use Rotation Vector=true
   ) {module=InitialMeshOrientation001}
   /SetVariables(Particles.CameraOffsetFlare=120) {isSetParameters=true}
   /SetVariables(Particles.SizeDark:Vector3f=<3 3 3>) {isSetParameters=true}
  ]
  particleUpdateScript [
   &ParticleState()
   &ScaleSpriteSize(
    Scale Factor=&MakeVector2D(
     X=&FloatFromCurve(
      FloatCurve=DataInterface() {propertyValues {Curve {keys [{arriveTangent=27.5059, interpMode=RCIM_Cubic, leaveTangent=27.5059, tangentMode=RCTM_User, time=0.1, value=0.1}, <at 0.2 1.1>]}}}
      Scale Curve=1.5
     )
     Y=&FloatFromCurve(
      FloatCurve=DataInterface() {propertyValues {Curve {keys [
          {arriveTangent=-55.1607, interpMode=RCIM_Cubic, leaveTangent=-55.1607, tangentMode=RCTM_User, time=0.02, value=3}
          {arriveTangent=-12.8632, interpMode=RCIM_Cubic, leaveTangent=-12.8632, tangentMode=RCTM_User, time=0.07, value=1.2}
          {interpMode=RCIM_Cubic, tangentMode=RCTM_User, time=0.3, value=0.2}
         ]}}}
      Scale Curve=1.5
     )
    )
    Scale Sprite Size Mode=::NewEnumerator2
   )
   &ScaleColor(
    Scale Alpha=&FloatFromCurve(
     FloatCurve=DataInterface() {propertyValues {Curve {keys [
         {arriveTangent=0.4784, interpMode=RCIM_Cubic, leaveTangent=0.4784, tangentMode=RCTM_User, time=0, value=0}
         {arriveTangent=-3.8452, interpMode=RCIM_Cubic, leaveTangent=-3.8452, tangentMode=RCTM_User, time=0.1, value=1}
         {arriveTangent=-0.1713, interpMode=RCIM_Cubic, leaveTangent=-0.1713, tangentMode=RCTM_User, time=1, value=0}
        ]}}}
     Scale Curve=1.3
    )
    Scale RGB=&MakeVectorFromLinearColorRGB(Color=&HueShiftLinearColor(
      Hue=&MakeCustomFloatFromBool(Bool=$System.IsWanderingBoss, True Float=-0.24, False Float=0)
      Saturation=&MakeCustomFloatFromBool(Bool=$System.IsWanderingBoss, True Float=0.7, False Float=1)
      Value=1
      Linear Color=&ColorFromCurve(ColorCurve=DataInterface() {propertyValues {
         AlphaCurve {keys [<at 0 1 RCIM_Cubic>]}
         BlueCurve {keys [<at 0.0169 50 RCIM_Cubic>,<at 0.6322 0.04>]}
         GreenCurve {keys [<at 0.0169 45.8333 RCIM_Cubic>,<at 0.6322 0.0133>]}
         RedCurve {keys [<at 0.0169 45 RCIM_Cubic>,<at 0.6322 0.008>]}
        }})
     ))
    ScaleA=true
    ScaleRGB=true
   )
   /SetVariables(Particles.ColorDark:LinearColor=&MakeLinearColorFromVectorAndFloat(
     "Vector (RGB)"=&MakeVectorFromLinearColorRGB(Color=&HueShiftLinearColor(
       Hue=&MakeCustomFloatFromBool(Bool=$System.IsWanderingBoss, True Float=-0.24, False Float=0)
       Saturation=&MakeCustomFloatFromBool(Bool=$System.IsWanderingBoss, True Float=0.7, False Float=1)
       Value=1
       Linear Color=&ColorFromCurve(ColorCurve=DataInterface() {propertyValues {
          AlphaCurve {keys [<at 0 1 RCIM_Cubic>]}
          BlueCurve {keys [<at 0 0.09 RCIM_Cubic>,<at 0.4371 0.016>]}
          GreenCurve {keys [<at 0 0 RCIM_Cubic>,<at 0.4371 0.0021>]}
          RedCurve {keys [<at 0 0.3 RCIM_Cubic>,<at 0.4371 0.0356>]}
         }})
      ))
     "Float (Alpha)"=&FloatFromCurve(
      FloatCurve=DataInterface() {propertyValues {Curve {keys [
          {interpMode=RCIM_Cubic, tangentMode=RCTM_User, time=0, value=0}
          {arriveTangent=-2.4988, interpMode=RCIM_Cubic, leaveTangent=-2.4988, tangentMode=RCTM_User, time=0.3, value=1}
          {interpMode=RCIM_Cubic, tangentMode=RCTM_User, time=1, value=0}
         ]}}}
      Scale Curve=3
     )
    )) {isSetParameters=true}
   &SolveForcesAndVelocity(Clamp Velocity=false, Limit Acceleration=false)
   &DynamicMaterialParameters(
    Index 0 Param 1=-0.1
    Index 0 Param 2=&Lerp_Float(
     A=0
     B=1
     Alpha=&FloatFromCurve(FloatCurve=DataInterface() {propertyValues {Curve {keys [<at 0 0>,<at 1 1>]}}})
    )
    Index 0 Param 3=0.5
    Index 0 Param 4=&Lerp_Float(
     A=0.45
     B=0.8
     Alpha=&FloatFromCurve(FloatCurve=DataInterface() {propertyValues {Curve {keys [
          {arriveTangent=0.2526, interpMode=RCIM_Cubic, leaveTangent=0.2526, tangentMode=RCTM_User, time=0, value=0}
          {arriveTangent=1.7785, interpMode=RCIM_Cubic, leaveTangent=1.7785, tangentMode=RCTM_User, time=1, value=1}
         ]}}})
    )
    Module.Index0Param0Write=true
    Module.Index0Param1Write=true
    Module.Index0Param2Write=true
    Module.Index0Param3Write=true
   )
   /SetVariables(Particles.DynamicMaterialParameterDark:Vector4f=&MakeVector4(
     X=-0.5
     Y=&Lerp_Float(
      A=0.2
      B=1.5
      Alpha=&FloatFromCurve(FloatCurve=DataInterface() {propertyValues {Curve {keys [<at 0 0>,<at 1 1>]}}})
     )
     Z=1
     W=&Lerp_Float(
      A=-0.5
      B=1
      Alpha=&FloatFromCurve(FloatCurve=DataInterface() {propertyValues {Curve {keys [
           {arriveTangent=0.2526, interpMode=RCIM_Cubic, leaveTangent=0.2526, tangentMode=RCTM_User, time=0.2, value=0}
           {arriveTangent=1.7785, interpMode=RCIM_Cubic, leaveTangent=1.7785, tangentMode=RCTM_User, time=1, value=1}
          ]}}})
     )
    )) {isSetParameters=true}
   /SetVariables(
    Particles.ColorFlare:LinearColor=&ColorFromCurve(ColorCurve=DataInterface() {propertyValues {
       AlphaCurve {keys [<at 0 0>,<at 0.0482 1 RCIM_Cubic>,<at 0.3413 0 RCIM_Cubic>]}
       BlueCurve {keys [<at 0.0373 1 RCIM_Cubic>,<at 0.1975 0.2 RCIM_Cubic>]}
       GreenCurve {keys [<at 0.0373 0.9167 RCIM_Cubic>,<at 0.1975 0.1833 RCIM_Cubic>]}
       RedCurve {keys [<at 0.0373 0.9 RCIM_Cubic>,<at 0.1975 0.18 RCIM_Cubic>]}
      }})
    Particles.PositionFlare:NiagaraPosition=&AddVectorToPosition(
     Position=$Particles.Position
     Vector=&MakeVector(
      X=0
      Y=0
      Z=&Lerp_Float(
       A=0
       B=20
       Alpha=&FloatFromCurve(FloatCurve=DataInterface() {propertyValues {Curve {keys [<at 0 0>,<at 0.65 1>]}}})
      )
     )
    )
   ) {isSetParameters=true}
   /SetVariables(
    Particles.ColorDark:LinearColor=&ColorFromCurve(ColorCurve=DataInterface() {propertyValues {
       AlphaCurve {keys [<at 0 0>,<at 0.0592 1.5 RCIM_Cubic>,<at 0.6507 0 RCIM_Cubic>]}
       BlueCurve {keys [<at 0 0.007 RCIM_Cubic>]}
       GreenCurve {keys [<at 0 0.0064 RCIM_Cubic>]}
       RedCurve {keys [<at 0 0.0063 RCIM_Cubic>]}
      }})
    Particles.SpriteSizeDark:Vector2f=&Vector2DFromFloat(Value=&Lerp_Float(
      A=0
      B=200
      Alpha=&FloatFromCurve(
       FloatCurve=DataInterface() {propertyValues {Curve {keys [
           {interpMode=RCIM_Cubic, tangentMode=RCTM_User, time=0, value=0}
           {interpMode=RCIM_Cubic, tangentMode=RCTM_User, time=0.1, value=1}
           {interpMode=RCIM_Cubic, tangentMode=RCTM_User, time=0.7, value=0}
          ]}}}
       Scale Curve=1.25
      )
     ))
   ) {isSetParameters=true}
   /SetVariables(Particles.SizeDark:Vector3f=&Multiply_VectorByFloat(
     Vector=$Particles.Initial.SizeDark
     Float=&FloatFromCurve(
      FloatCurve=DataInterface() {propertyValues {Curve {keys [
          {interpMode=RCIM_Cubic, tangentMode=RCTM_Break, time=0, value=0.2}
          {arriveTangent=2.9365, interpMode=RCIM_Cubic, leaveTangent=2.9365, tangentMode=RCTM_User, time=0.1, value=1}
          {interpMode=RCIM_Cubic, tangentMode=RCTM_Break, time=1, value=2}
         ]}}}
      Scale Curve=0.85
     )
    )) {isSetParameters=true}
   /SetVariables(Particles.ColorMeshDark:LinearColor=&MultiplyLinearColor(
     Linear Color=<rgba 0.036 0.0363 0.04 5>
     Scale Factor=&MakeFloat4FromVectorAndFloat(
      "Vector (RGB)"=&VectorFromFloat(Value=&FloatFromCurve(FloatCurve=DataInterface() {propertyValues {Curve {keys [
            {interpMode=RCIM_Cubic, tangentMode=RCTM_Break, time=0, value=10}
            {arriveTangent=-6.6072, interpMode=RCIM_Cubic, leaveTangent=-6.6072, tangentMode=RCTM_User, time=0.15, value=3}
            {interpMode=RCIM_Cubic, tangentMode=RCTM_Break, time=0.8, value=1}
           ]}}}))
      "Float (Alpha)"=&FloatFromCurve(
       FloatCurve=DataInterface() {propertyValues {Curve {keys [
           {arriveTangent=-3.1255, interpMode=RCIM_Cubic, leaveTangent=-3.1255, tangentMode=RCTM_User, time=0.6, value=1}
           {interpMode=RCIM_Cubic, tangentMode=RCTM_User, time=1, value=0}
          ]}}}
       Scale Curve=&Lerp_Float(A=0.1, B=0.25, Alpha=$System.DistanceScalarNear)
      )
     )
    )) {isSetParameters=true}
   /SetVariables(Particles.DMP_01:Vector4f=&MakeVector4(
     X=&FloatFromCurve(FloatCurve=DataInterface() {propertyValues {Curve {keys [{arriveTangent=-2.1015, interpMode=RCIM_Cubic, leaveTangent=-2.1015, tangentMode=RCTM_User, time=0, value=0.5}, <at 1 -0.15>]}}})
     Y=&FloatFromCurve(FloatCurve=DataInterface() {propertyValues {Curve {keys [{arriveTangent=-0.3464, interpMode=RCIM_Cubic, leaveTangent=-0.3464, tangentMode=RCTM_User, time=0, value=1}, <at 1 0.9>]}}})
     Z=&FloatFromCurve(FloatCurve=DataInterface() {propertyValues {Curve {keys [
          {interpMode=RCIM_Cubic, tangentMode=RCTM_Break, time=0, value=0}
          {arriveTangent=0.4478, interpMode=RCIM_Cubic, leaveTangent=0.4478, tangentMode=RCTM_User, time=0.15, value=1}
          {interpMode=RCIM_Cubic, tangentMode=RCTM_Break, time=1, value=1.2}
         ]}}})
     W=1
    )) {isSetParameters=true}
  ]
  rendererClasses [&NiagaraSpriteRendererProperties, &NiagaraMeshRendererProperties]
  renderers [
   {rendererClass=&NiagaraSpriteRendererProperties}
   {rendererClass=&NiagaraSpriteRendererProperties, rendererIndex=1}
   {rendererClass=&NiagaraMeshRendererProperties, rendererIndex=2}
  ]
 }
 SmokeLoop_LW {rendererClasses [], renderers []}
}
summary {
 systemName=NS_FossilMeal_Bread_Leaps
 emitters [
  {emitterName=DarkStreaks, rendererClasses [&NiagaraSpriteRendererProperties]}
  {bEnabled=false, emitterName=DarkBits, rendererClasses [&NiagaraSpriteRendererProperties]}
  {emitterName=FlareSmoke, rendererClasses [&NiagaraSpriteRendererProperties, &NiagaraMeshRendererProperties]}
  {emitterName=SmokeLoop_LW, rendererClasses []}
 ]
 userVariables [
  User.ExplosionScale {type=NiagaraFloat, defaultValue=1, description=""}
  User.Scalar {type=NiagaraFloat, defaultValue=2.5, description=""}
 ]
}
