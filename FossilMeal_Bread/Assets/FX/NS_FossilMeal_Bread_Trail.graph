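# Niagara ribbon trail for FossilMeal Bread, emitting short-lived dark blue-grey particles that drift upward.
# System-level cloak alpha scales particle opacity, combined with fades over age and normalized execution index to soften the ribbon.
# A normalized velocity value scales ribbon width on top of an age curve, so width follows speed.
# Age curves also drive dynamic material parameters for the ribbon's look.
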NiagaraSystem /FossilMeal_Bread/Assets/FX/NS_FossilMeal_Bread_Trail
dependencies {
 usedDataInterfaces [&NiagaraDataInterfaceCurve]
 usedDynamicInputs [
  &Lerp_Float
  &ClampFloat
  &RemapRange
  &VectorLength
 ]
 usedModules [
  &SystemState
  &Timeline
  &self:SystemSpawnScript.SystemScriptSource.SystemScriptGraph.NiagaraNodeAssignment_2.SetVariables_#1
  &InterpolateOverTime
  &self:SystemSpawnScript.SystemScriptSource.SystemScriptGraph.NiagaraNodeAssignment_0.SetVariables_#2
  &self:SystemSpawnScript.SystemScriptSource.SystemScriptGraph.NiagaraNodeAssignment_1.SetVariables_#3
 ]
}
emitters {Trail {
  emitterUpdateScript [
   &EmitterState()
   &SpawnRate(SpawnRate=55, Use Spawn Probability=false)
  ]
  particleSpawnScript [&InitializeParticle(
    Color=&MultiplyLinearColor(
     Linear Color=<rgba 0.0521 0.0665 0.0938 1>
     Scale Factor=&MakeVector4(X=1, Y=1, Z=1, W=$System.AlphabIsCloaked)
    )
    Lifetime=0.5
    Mass Mode=::NewEnumerator3
    Ribbon Width=200
    Ribbon Width Mode=::NewEnumerator1
    UsePositionOffset=false
    Write Material Random=true
   )]
  particleUpdateScript [
   &ParticleState()
   &GravityForce(Gravity=<0 0 90>)
   &ScaleColor(
    Scale Alpha=&Multiply_Float(
     A=$System.AlphabIsCloaked
     B=&Multiply_Float(
      A=&FloatFromCurve(FloatCurve=DataInterface() {propertyValues {Curve {keys [{interpMode=RCIM_Cubic, tangentMode=RCTM_User, time=0, value=1}, {interpMode=RCIM_Cubic, tangentMode=RCTM_User, time=1, value=0}]}}})
      B=&FloatFromCurve(
       FloatCurve=DataInterface() {propertyValues {Curve {keys [
           {arriveTangent=10.903, interpMode=RCIM_Cubic, leaveTangent=10.903, tangentMode=RCTM_User, time=0, value=0}
           {arriveTangent=2, interpMode=RCIM_Cubic, leaveTangent=2, time=0.1386, value=0.8718}
           {interpMode=RCIM_Cubic, tangentMode=RCTM_User, time=0.5, value=1}
          ]}}}
       CurveIndex=&ReturnNormalizedExecIndex()
      )
     )
    )
    ScaleA=true
    ScaleRGB=false
   ) {module=ScaleColor001}
   &SolveForcesAndVelocity(Clamp Velocity=false, Limit Acceleration=false)
   &ScaleRibbonWidth(Ribbon Width Scale=&FloatFromCurve(
     FloatCurve=DataInterface() {propertyValues {Curve {keys [
         {arriveTangent=19.2206, interpMode=RCIM_Cubic, leaveTangent=19.2206, tangentMode=RCTM_User, time=0, value=0.2}
         {arriveTangent=2.25, interpMode=RCIM_Cubic, leaveTangent=2.25, time=0.1, value=2}
         <at 0.3 0.875 RCIM_Cubic>
         {arriveTangent=0.3186, interpMode=RCIM_Cubic, leaveTangent=0.3186, tangentMode=RCTM_User, time=0.7, value=2}
        ]}}}
     Scale Curve=&FloatFromCurve(
      FloatCurve=DataInterface() {propertyValues {Curve {
         defaultValue=1
         keys [{interpMode=RCIM_Cubic, tangentMode=RCTM_User, time=0, value=0}, {interpMode=RCIM_Cubic, tangentMode=RCTM_User, time=1, value=1}]
        }}}
      CurveIndex=$System.NormalizedVelocityRange
     )
    ))
   &DynamicMaterialParameters(
    Index 0 Param 2=&FloatFromCurve(FloatCurve=DataInterface() {propertyValues {Curve {keys [{tangentMode=RCTM_User, time=0, value=0}, {tangentMode=RCTM_User, time=1, value=1}]}}})
    Index 0 Param 4=&FloatFromCurve(FloatCurve=DataInterface() {propertyValues {Curve {keys [
         {interpMode=RCIM_Cubic, leaveTangent=-48.5828, tangentMode=RCTM_Break, time=0, value=3}
         {arriveTangent=-2.4798, tangentMode=RCTM_User, time=0.2, value=0.25}
         <at 0.5 0>
        ]}}})
    Module.Index0Param0Write=true
    Module.Index0Param1Write=true
    Module.Index0Param2Write=true
    Module.Index0Param3Write=true
    Module.Index1Param0Write=false
    Module.Index1Param1Write=false
    Module.Index1Param2Write=false
    Module.Index1Param3Write=false
    Write Parameter Index 1=true
   )
  ]
  rendererClasses [&NiagaraRibbonRendererProperties]
  renderers [{rendererClass=&NiagaraRibbonRendererProperties}]
 }}
summary {
 systemName=NS_FossilMeal_Bread_Trail
 emitters [{emitterName=Trail, rendererClasses [&NiagaraRibbonRendererProperties]}]
 userVariables [
  User.bIsCloaked {type=NiagaraBool, defaultValue=true}
  User.Scalar {type=NiagaraFloat, defaultValue=2, description=""}
 ]
}
